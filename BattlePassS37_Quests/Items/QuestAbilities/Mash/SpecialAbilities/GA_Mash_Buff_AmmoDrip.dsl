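; Gameplay ability for a Mash buff that periodically drips random ammo into the owning player's inventory.
; On activation, where cosmetic events cannot execute, it starts a looping timer with a tick rate scaled by ability level.
; Each tick picks one of five ammo types by weighted chance.
; It grants a level-scaled amount, skipping the tick if that type is at its cap or the pawn is invalid.
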(event Ability|EventActivateAbility
  (Networking|CanExecuteCosmeticEvents
    (:False
      (bind _returnvalue (Utilities|Time|SetTimerbyEvent (delegate Custom|AmmoGrant) (ScalableFloat|GetValueatLevel (Variables|Default|GetTickRate) (Math|Conversions|ToFloat(Integer) (Ability|GetAbilityLevel))) true))
      (Variables|Default|SetAmmoGrantTimerHandle _returnvalue))))

(event Custom|AmmoGrant
  (bind _output_get (Variables|Default|SetAmmoToGrant (Utilities|Array|Get(acopy) (Variables|Default|GetAmmoSelector) (Math|Fortnite|GetRandomWeightedIndex (Utilities|Array|MakeArray (ScalableFloat|GetValueatLevel (Variables|Default|GetWeightedChances)) (ScalableFloat|GetValueatLevel (Variables|Default|GetWeightedChances) 1.0) (ScalableFloat|GetValueatLevel (Variables|Default|GetWeightedChances) 2.0) (ScalableFloat|GetValueatLevel (Variables|Default|GetWeightedChances) 3.0) (ScalableFloat|GetValueatLevel (Variables|Default|GetWeightedChances) 4.0))))))
  (if (not (CallFunction|HasFullAmmo (Utilities|Map|Find (Variables|Default|GetMaxAmmoCap) _output_get)))
    (bind _asfort_world_item_definition (Utilities|Casting|CastToFortWorldItemDefinition (Variables|Default|GetAmmoToGrant)))
    (Inventory|GiveItemtoInventoryOwner :InventoryOwner (Fortnite|GetFortPlayerControllerfromActor (CallFunction|GetAvatarasFortPlayerPawn)) :ItemDefinition _asfort_world_item_definition :NumberToGive (ScalableFloat|GetValueasInteger (Utilities|Map|Find (Variables|Default|GetAmmoAmountToGive) _asfort_world_item_definition) (Math|Conversions|ToFloat(Integer) (Ability|GetAbilityLevel))) :bNotifyPlayer true :ItemLevel -1 :PickupInstigatorHandle -1 :bUseItemPickupAnalyticEvent false :WeaponAmmoOverride -1)))

(fn HasFullAmmo
  (params (AmmoChosen :type Integer))
  (bind _outfortplayerpawn (CallFunction|GetAvatarasValidFortPlayerPawn)
    (:ValidFortPlayerPawn
      (return (>= (Inventory|GetItemQuantityOnPlayer (Fortnite|GetFortPlayerControllerfromActor _outfortplayerpawn) (Variables|Default|GetAmmoToGrant)) AmmoChosen)))
    (:AvatarCastFailed
      (return true))))
