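; Gameplay ability that regenerates the player's health, then shield, up to level-scaled percentage caps.
; If health needs no healing at setup, it starts damage-event and attribute-change tasks that toggle handle-tracked regen GameplayEffects.
; Damage sets a flag that pauses regen until a three-second timer retriggers it; storms also block health healing.
; Pawn possession reruns setup after respawn, and ending the ability removes both effects.
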(event Custom|PlayerRespawned
  (params (Pawn :type Pawn%20Object%20Reference))
  (bind _playerpawn (Variables|Default|GetPlayerPawn))
  (bind _asfort_player_pawn (Utilities|Casting|CastToFortPlayerPawn Pawn))
  (Variables|Default|SetPlayerPawn _asfort_player_pawn)
  (if (CallFunction|ShouldWeHealHealth)
    (Variables|Default|SetbIsFullHealth false)
    (else
      (Variables|Default|SetbIsFullHealth true)
      (if (CallFunction|ShouldWeHealShield)
        (Utilities|Time|SetTimerbyEvent (EventDispatchers|CreateEvent :!event Custom|RetriggerHeal) 3.0))
      (Ability|Tasks|WaitGameplayEvent "(TagName=\"Event.Damage.Shielded\")"
        (:then
          (Ability|Tasks|WaitGameplayEvent "(TagName=\"Event.Damage.Received\")"
            (:then
              (Ability|Tasks|WaitForAttributeChange :Attribute "(AttributeName=\"CurrentShield\",Attribute=/Script/FortniteGame.FortHealthSet:CurrentShield,AttributeOwner=\"/Script/CoreUObject.Class'/Script/FortniteGame.FortHealthSet'\")" :TriggerOnce false :OptionalExternalOwner _playerpawn
                (:then
                  (Ability|Tasks|WaitForAttributeChange :Attribute "(AttributeName=\"Health\",Attribute=/Script/FortniteGame.FortHealthSet:Health,AttributeOwner=\"/Script/CoreUObject.Class'/Script/FortniteGame.FortHealthSet'\")" :TriggerOnce false :OptionalExternalOwner _playerpawn
                    (:OnChange
                      (if (CallFunction|ShouldWeHealHealth)
                        (CallFunction|ApplyHealthHealing)
                        (else
                          (CallFunction|RemoveHealthHealing))))))
                (:OnChange
                  _))
              (if (CallFunction|ShouldWeHealShield)
                (CallFunction|ApplyShieldHeal)
                (else
                  (CallFunction|RemoveShieldHealing))))
            (:EventReceived
              _)))
        (:EventReceived
          _))
      (CallFunction|RemoveShieldHealing)
      (Variables|Default|SetRecentlyDamaged true)
      (Utilities|Time|SetTimerbyEvent (delegate Custom|RetriggerHeal) 3.0))))

(event Ability|EventActivateAbility
  (bind _playerpawn (Variables|Default|GetPlayerPawn))
  (bind _outfortplayerpawn (CallFunction|GetAvatarasValidFortPlayerPawn)
    (:ValidFortPlayerPawn
      (Variables|Default|SetPlayerPawn _outfortplayerpawn)
      (Game|Player|BindEventtoOnPlayerPawnPossessed (Fortnite|GetFortPlayerControllerfromActor _outfortplayerpawn) (EventDispatchers|CreateEvent :!event Custom|PlayerRespawned))
      (if (CallFunction|ShouldWeHealHealth)
        (Variables|Default|SetbIsFullHealth false)
        (else
          (Variables|Default|SetbIsFullHealth true)
          (if (CallFunction|ShouldWeHealShield)
            (Utilities|Time|SetTimerbyEvent (EventDispatchers|CreateEvent :!event Custom|RetriggerHeal) 3.0))
          (Ability|Tasks|WaitGameplayEvent "(TagName=\"Event.Damage.Shielded\")"
            (:then
              (Ability|Tasks|WaitGameplayEvent "(TagName=\"Event.Damage.Received\")"
                (:then
                  (Ability|Tasks|WaitForAttributeChange :Attribute "(AttributeName=\"CurrentShield\",Attribute=/Script/FortniteGame.FortHealthSet:CurrentShield,AttributeOwner=\"/Script/CoreUObject.Class'/Script/FortniteGame.FortHealthSet'\")" :TriggerOnce false :OptionalExternalOwner _playerpawn
                    (:then
                      (Ability|Tasks|WaitForAttributeChange :Attribute "(AttributeName=\"Health\",Attribute=/Script/FortniteGame.FortHealthSet:Health,AttributeOwner=\"/Script/CoreUObject.Class'/Script/FortniteGame.FortHealthSet'\")" :TriggerOnce false :OptionalExternalOwner _playerpawn
                        (:OnChange
                          (if (CallFunction|ShouldWeHealHealth)
                            (CallFunction|ApplyHealthHealing)
                            (else
                              (CallFunction|RemoveHealthHealing))))))
                    (:OnChange
                      _))
                  (if (CallFunction|ShouldWeHealShield)
                    (CallFunction|ApplyShieldHeal)
                    (else
                      (CallFunction|RemoveShieldHealing))))
                (:EventReceived
                  _)))
            (:EventReceived
              _))
          (CallFunction|RemoveShieldHealing)
          (Variables|Default|SetRecentlyDamaged true)
          (Utilities|Time|SetTimerbyEvent (delegate Custom|RetriggerHeal) 3.0))))))

(event Custom|ApplyShieldHeal
  (bind _returnvalue (Ability|GameplayEffect|GetActiveGameplayEffectStackCount (Variables|Default|GetShieldGE)))
  (if (<= _returnvalue 0)
    (bind _returnvalue_2 (Ability|ApplyGameplayEffectToOwner "/BattlePassS37_Quests/Items/QuestAbilities/Mash/SpecialAbilities/GE_Mash_Buff_ShieldRegen.GE_Mash_Buff_ShieldRegen_C" (Ability|GetAbilityLevel)))
    (Variables|Default|SetShieldGE _returnvalue_2)))

(event Custom|RemoveShieldHealing
  (Ability|RemoveGameplayEffectFromOwnerWithHandle (Variables|Default|GetShieldGE)))

(event Custom|RetriggerHeal
  (Variables|Default|SetRecentlyDamaged false)
  (if (CallFunction|ShouldWeHealHealth)
    (CallFunction|ApplyHealthHealing)
    (elif (CallFunction|ShouldWeHealShield)
      (CallFunction|ApplyShieldHeal))))

(event Custom|ApplyHealthHealing
  (bind _returnvalue (Ability|GameplayEffect|GetActiveGameplayEffectStackCount (Variables|Default|GetHealGE)))
  (if (<= _returnvalue 0)
    (Variables|Default|SetbIsFullHealth false)
    (bind _returnvalue_2 (Ability|ApplyGameplayEffectToOwner "/BattlePassS37_Quests/Items/QuestAbilities/Mash/SpecialAbilities/GE_Mash_Buff_HealthRegen.GE_Mash_Buff_HealthRegen_C" (Ability|GetAbilityLevel)))
    (Variables|Default|SetHealGE _returnvalue_2)))

(event Custom|RemoveHealthHealing
  (Ability|RemoveGameplayEffectFromOwnerWithHandle (Variables|Default|GetHealGE))
  (Variables|Default|SetbIsFullHealth true))

(event Ability|EventOnEndAbility
  (params (bWasCancelled :type Boolean))
  (|CallParentFunction bWasCancelled)
  (CallFunction|RemoveShieldHealing)
  (CallFunction|RemoveHealthHealing)
  (bind _playerpawn (Variables|Default|GetPlayerPawn)
    (:then
      (Game|Player|BindEventtoOnPlayerPawnPossessed (Fortnite|GetFortPlayerControllerfromActor _playerpawn) (EventDispatchers|CreateEvent :!event Custom|PlayerRespawned)))
    (:else)))

(fn ShouldWeHealHealth
  (return (Math|Boolean|ANDBoolean (not (Variables|/Script/FortniteGame.FortPlayerPawn|GetbIsInAnyStorm (Variables|Default|GetPlayerPawn))) (< (FortPawn|GetHealth (Variables|Default|GetPlayerPawn)) (* (FortPawn|GetMaxHealth (Variables|Default|GetPlayerPawn)) (ScalableFloat|GetValueatLevel (Variables|Default|GetPercentOfMaxHealthCap) (Math|Conversions|ToFloat(Integer) (Ability|GetAbilityLevel))))) (not (Variables|Default|GetRecentlyDamaged)))))

(fn ShouldWeHealShield
  (return (Math|Boolean|ANDBoolean (< (FortPawn|GetShield (Variables|Default|GetPlayerPawn)) (* (FortPawn|GetMaxShield (Variables|Default|GetPlayerPawn)) (ScalableFloat|GetValueatLevel (Variables|Default|GetPercentOfMaxShieldCap) (Math|Conversions|ToFloat(Integer) (Ability|GetAbilityLevel))))) (not (Variables|Default|GetRecentlyDamaged)) (not (CallFunction|ShouldWeHealHealth)))))
